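# Unlit translucent beam-trail material for the Arctic Sniper pickaxe's melee trail, with a scrolling cyan, blue and warm gradient along U.
# The gradient multiplies a panning trace texture, is partly posterized with Round, then modulated by a noise-distorted cutout texture.
# That result drives emissive with particle color, tint and desaturation.
# It also drives opacity through a Dynamic Parameter dissolve, sine edge masks, non-mobile depth fade and night-time dimming.
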Material /BRCosmetics/Weapons/FORT_Melee/Pickaxe_ArcticSniper/FX/M_Melee_Arctic_Sniper_Trail
graph {
 properties {bUsedWithBeamTrails=true, blendMode=BLEND_Translucent, pixelDepthOffsetMode=PDOM_Legacy, shadingModel=MSM_Unlit, usedShadingModels=Unlit}
 nodes [
  n1 = ParticleColor()
  n2 = Add(A=MaterialFunctionCall(Float2=TextureCoordinate()) {materialFunction=&BreakOutFloat2Components}.R) {constB=-0.2}
  n3 = LinearInterpolate(A=Constant3Vector() {constant=<rgba 0.0615 0.8602 1 0>}, B=Constant3Vector() {constant=<rgba 0 0.137 1 0>}, Alpha=@n2)
  n4 = Constant3Vector() {constant=<rgba 1 0.1825 0.1228 0>}
  n5 = Constant3Vector() {constant=<rgba 1 0.3154 0.0118 0>}
  n6 = MaterialFunctionCall(Float2=TextureCoordinate()) {materialFunction=&BreakOutFloat2Components}
  n7 = TextureCoordinate() {uTiling=0.5}
  n8 = Multiply(
   A=LinearInterpolate(A=LinearInterpolate(A=@n3, B=@n4, Alpha=@n2), B=Multiply(A=@n5), Alpha=Add(A=@n6.R) {constB=-0.4})
   B=TextureSampleParameter2D!(UVs=Panner(Coordinate=@n7) {speedX=1}, Bias, Apply View MipBias) {mipValueMode=TMVM_MipBias, parameterName=Texture_A, texture=/Game/Effects/Fort_Effects/Textures/Smoke/T_ShotGun_Trace_Packed}.RGB
  )
  n9 = TextureSample(UVs=Panner(Coordinate=TextureCoordinate() {uTiling=3, vTiling=0.5}) {speedY=0.5}) {samplerType=SAMPLERTYPE_Grayscale, texture=&LowResBlurredNoise}
  n10 = Panner(
   Coordinate=Add(A=TextureCoordinate(), B=Multiply(A=@n9.R) {constB=0.2})
   Speed=AppendVector(A=ScalarParameter() {defaultValue=1, parameterName=Texture_B_Speed_U}, B=ScalarParameter() {parameterName=Texture_B_Speed_V})
  ) {speedX=1}
  n11 = Multiply(
   A=LinearInterpolate(A=Round(None=@n8), B=@n8, Alpha=Constant() {r=0.7})
   B=Multiply(A=TextureSampleParameter2D!(UVs=@n10, Apply View MipBias) {parameterName=Texture_B, texture=/Game/Effects/Fort_Effects/Textures/Misc/T_Shot0030_NoiseCutout}.G) {constB=2}
  )
  n12 = Multiply(A=@n11)
  n13 = TextureCoordinate()
  n14 = AppendVector(A=ScalarParameter() {defaultValue=3, parameterName=Dissolve_UTiling}, B=ScalarParameter() {defaultValue=0.5, parameterName=Dissolve_VTiling})
  n15 = TextureSampleParameter2D(UVs=Panner(Coordinate=Multiply(A=@n13, B=@n14)) {speedX=0.5}) {parameterName=Dissolve_Texture, samplerType=SAMPLERTYPE_LinearGrayscale, texture=/Game/Global/Textures/Default/Noise/T_SwirlNoise_N_height}
  n16 = ScalarParameter() {defaultValue=1, parameterName=Dissolve_brightness}
  n17 = Multiply(A=MaterialFunctionCall!(Ramp=DynamicParameter() {defaultValue=<rgba 0.8 1 1 1>}.Param1, Dissolve=Multiply(A=@n15.G, B=@n16), EdgeSharpness) {materialFunction=/Game/Effects/Fort_Effects/Materials/_Functions/MF_Dissolve}.ClampedResult)
  n18 = MaterialFunctionCall(Float2=TextureCoordinate() {vTiling=0.99}) {materialFunction=&BreakOutFloat2Components}
  n19 = Sine(None=@n18.R) {period=2}
  n20 = Sine(None=@n18.G) {period=2}
  n21 = Multiply(A=Multiply(A=@n1.A, B=Multiply(A=@n12, B=@n17)), B=Multiply(A=Power(Base=@n19), B=Power(Base=@n20) {constExponent=2.5}))
  n22 = DepthFade(Opacity=@n21, FadeDistance=ScalarParameter() {defaultValue=100, parameterName=Depth_Fade})
  n23 = ShadingPathSwitch(Default=@n22, Deferred=@n22, Mobile=@n21) {desc=Migrated from FeatureLevelSwitch}
  n24 = Multiply(A=@n23)
  MaterialFunctionCall(Color_Morning=@n24, Color_Day=@n24, Color_Evening=@n24, Color_Night=Multiply(A=@n23) {constB=0.7}) {materialFunction=&Color_Time_of_Day}
  n25 = Multiply(A=@n1.RGB, B=Multiply(A=@n11))
  n26 = VectorParameter() {parameterName=Emissive_Color, defaultValue=<rgba 0.8646 0.8646 0.8646 0>}
  MaterialFunctionCall(Emissive=Desaturation(None=Multiply(A=@n25, B=@n26.RGB), Fraction=ScalarParameter() {parameterName=Desaturation_Intensity})) {materialFunction=&MF_FX_EmissiveAdjustments}
 ]
}
