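# Translucent unlit mesh-particle FX material for the Arctic Sniper pickaxe, tinted by particle color and a color parameter.
# Two panning textures, optionally distorted by scrolling turbulence, multiply into opacity, shaped by a V-gradient mask, optional axis-aligned fresnel edge blur and depth fade.
# Index-of-refraction distortion combines a sine V mask, fresnel and a dynamic "Reflection" parameter, alongside a panning normal map.
# Mobile skips depth fade and uses a constant refraction of 1.
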Material /BRCosmetics/Weapons/FORT_Melee/Pickaxe_ArcticSniper/FX/M_Melee_Arctic_Sniper_Weapon
graph {
 properties {
  bUsedWithMeshParticles=true
  blendMode=BLEND_Translucent
  pixelDepthOffsetMode=PDOM_Legacy
  refractionMethod=RM_IndexOfRefraction
  shadingModel=MSM_Unlit
  usedShadingModels=Unlit
 }
 nodes [
  n1 = ParticleColor()
  n2 = VectorParameter() {parameterName=Texture_color, defaultValue=<rgba 1 1 1 0>}
  MaterialFunctionCall(Emissive=Desaturation(None=Multiply(A=@n1.RGB, B=@n2.RGB), Fraction=ScalarParameter() {parameterName=Desaturation_Intensity})) {materialFunction=&MF_FX_EmissiveAdjustments}
  n3 = Multiply(
   A=TextureCoordinate()
   B=AppendVector(A=ScalarParameter() {defaultValue=3, parameterName=Texture_A_UTiling}, B=ScalarParameter() {defaultValue=1, parameterName=Texture_A_VTiling})
  )
  n4 = TextureCoordinate() {uTiling=2}
  n5 = Multiply(
   A=TextureSample(UVs=Panner(Coordinate=@n4) {speedX=1, speedY=2}) {texture=/Game/Effects/Fort_Effects/Textures/Power_Source/T_Power_Source_Sphere_Texture}.R
   B=ScalarParameter() {defaultValue=0.3, parameterName=Turbulence}
  ) {constB=0.3}
  n6 = Add(A=@n3, B=@n5)
  n7 = ScalarParameter() {parameterName=Texture_A_speedU}
  n8 = ScalarParameter() {defaultValue=2, parameterName=Texture_A_speedv}
  n9 = TextureSampleParameter2D!(
   UVs=Panner(Coordinate=StaticSwitchParameter(True=@n6, False=@n3) {parameterName=Disturbance_on_A}, Speed=AppendVector(A=@n7, B=@n8))
   Bias=ScalarParameter() {parameterName=Bias_Intensity}
   Apply View MipBias
  ) {
   mipValueMode=TMVM_MipBias
   parameterName=Texture_A
   samplerType=SAMPLERTYPE_LinearColor
   texture=/Game/Effects/Fort_Effects/BluGlow/Violentium/Violentium_Mask
  }
  n10 = Multiply(
   A=TextureCoordinate()
   B=AppendVector(
    A=ScalarParameter() {defaultValue=1, parameterName=Texture_B_UTiling}
    B=ScalarParameter() {defaultValue=0.5, parameterName=Texture_B_VTiling}
   )
  )
  n11 = Panner(
   Coordinate=StaticSwitchParameter(True=Add(A=@n10, B=@n5), False=@n10) {parameterName=Disturbance_on_B}
   Speed=AppendVector(A=ScalarParameter() {parameterName=Texture_B_speedU}, B=ScalarParameter() {defaultValue=0.5, parameterName=Texture_B_speedV})
  ) {speedX=1}
  n12 = Multiply(
   A=Multiply(A=Multiply(A=@n9.RGB), B=ScalarParameter() {defaultValue=1, parameterName=Texture_A_intensity})
   B=Multiply(
    A=TextureSampleParameter2D!(UVs=@n11, Apply View MipBias) {parameterName=Texture_B, texture=/Game/Global/Textures/Default/Masks_and_Gradients/T_BlasterBeamDistort}.RGB
    B=ScalarParameter() {defaultValue=2, parameterName=Texture_B_intensity}
   ) {constB=2}
  )
  n13 = ScalarParameter() {defaultValue=30, parameterName=Texture_Intensity}
  n14 = Multiply(A=@n12)
  n15 = MaterialFunctionCall(Float2=TextureCoordinate()) {materialFunction=&BreakOutFloat2Components}
  n16 = Multiply(
   A=Power(Base=@n15.G, Exp=ScalarParameter() {defaultValue=2, parameterName=Mask_Power_a})
   B=Power(Base=OneMinus(None=@n15.G), Exp=ScalarParameter() {defaultValue=2, parameterName=Mask_Power_b})
  )
  n17 = Transform(None=Constant3Vector() {constant=<rgba 0 1 0 0>}) {transformType=TRANSFORM_Local}
  n18 = ScalarParameter() {defaultValue=3, parameterName=Dot_intensity}
  n19 = Multiply(
   A=Clamp(None=Multiply(A=@n16) {constB=10})
   B=StaticSwitchParameter(
    True=MaterialFunctionCall!(Axis=@n17, Exponent=@n18, Normal) {materialFunction=/Game/Effects/Fort_Effects/Materials/_Functions/MF_AxisAlignedFrensnel}.Fresnel
    False=Constant() {r=1}
   ) {defaultValue=true, parameterName=Edge_blur_ON}
  )
  n20 = Multiply(A=@n1.A, B=Multiply(A=Multiply(A=@n12, B=@n13) {constB=30}, B=Multiply(A=@n14, B=@n19)))
  n21 = DepthFade(Opacity=@n20, FadeDistance=ScalarParameter() {defaultValue=100, parameterName=Depth_Fade})
  ShadingPathSwitch(Default=@n21, Deferred=@n21, Mobile=@n20) {desc=Migrated from FeatureLevelSwitch}
  n22 = MaterialFunctionCall(Float2=TextureCoordinate()) {materialFunction=&BreakOutFloat2Components}
  n23 = Multiply(B=Power(Base=Sine(None=@n22.G) {period=2})) {constA=1}
  n24 = MaterialFunctionCall(
   Axis=Transform(None=Constant3Vector() {constant=<rgba 0 1 0 0>}) {transformType=TRANSFORM_Local}
   Exponent=ScalarParameter() {defaultValue=3, parameterName=Dot_intensity}
  ) {materialFunction=&MF_AxisAlignedFrensnel}
  n25 = LinearInterpolate(Alpha=Multiply(
    A=Multiply(A=@n23, B=@n24.Fresnel)
    B=DynamicParameter() {defaultValue=<rgba 5 1 1 1>, paramNames [Reflection,Param2,Param3,Param4]}.Reflection
   )) {constA=1, constB=1.05}
  ShadingPathSwitch(Default=@n25, Deferred=@n25, Mobile=Constant() {r=1}) {desc=Migrated from FeatureLevelSwitch}
  n26 = TextureCoordinate()
  n27 = AppendVector(
   A=ScalarParameter() {defaultValue=1, parameterName=Reflection_U_Tiling}
   B=ScalarParameter() {defaultValue=0.5, parameterName=Reflection_V_Tiling}
  )
  TextureSampleParameter2D(UVs=Panner(Coordinate=Multiply(A=@n26, B=@n27)) {speedY=1}) {
   parameterName=Reflection_Texture
   samplerType=SAMPLERTYPE_Normal
   texture=/Game/Effects/Fort_Effects/Gameplay/SmokeDustExplosions/Textures/T_Explosion_NM_01
  }
 ]
}
