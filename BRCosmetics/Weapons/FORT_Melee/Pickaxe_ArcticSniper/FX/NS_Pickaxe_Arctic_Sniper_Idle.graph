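# Niagara idle effect for the Arctic Sniper pickaxe, with both emitters positioned through a system-level socket reader.
# Fire continuously spawns short-lived, glowing blue sprites in a cone, shaped by drag and curves for size, alpha and a dynamic material parameter.
# background001 bursts persistent mesh particles split into six partitions, each selecting its color, scale and brightness, and setting a visibility tag across six mesh renderers to layer the glow.
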NiagaraSystem /BRCosmetics/Weapons/FORT_Melee/Pickaxe_ArcticSniper/FX/NS_Pickaxe_Arctic_Sniper_Idle
dependencies {
 usedDataInterfaces [&NiagaraDataInterfaceCurve, &NiagaraDataInterfaceSocketReader]
 usedDynamicInputs []
 usedModules [
  &self:SystemSpawnScript.SystemScriptSource.SystemScriptGraph.NiagaraNodeAssignment_0.SetVariables_#1
  &SystemState
 ]
}
emitters {
 Fire {
  emitterUpdateScript [
   &EmitterState(
    Inactive Response=::NewEnumerator1
    Life Cycle Mode=::NewEnumerator1
    Loop Behavior=::NewEnumerator1
    Loop Duration Mode=::NewEnumerator1
    UseLoopDelay=true
   )
   &SpawnRate(SpawnRate=30, Use Spawn Probability=false)
  ]
  particleSpawnScript [
   &InitializeParticle(
    Lifetime=0.2
    Sprite Rotation Angle=&RandomRangeFloat(Minimum=0, Maximum=1, Evaluation Type=::NewEnumerator1)
    Sprite Rotation Mode=::NewEnumerator1
    Sprite Size=&Vector2DFromFloat(Value=&RandomRangeFloat(Minimum=4, Maximum=7, Evaluation Type=::NewEnumerator1))
    Sprite Size Mode=::NewEnumerator4
    UsePositionOffset=false
    Write Material Random=true
   )
   &Color(Color=&MakeLinearColorFromVectorAndFloat("Vector (RGB)"=<0.0425 0.4647 3>, "Float (Alpha)"=1))
   &SocketLocation(SamplePosOffset=false, Socket Reader=$System.Socket Reader, Socket Sampling Mode=::NewEnumerator3, VelocitySpeedLimit=false)
   &AddVelocity(
    Cone Angle=80
    Cone Angle Mode:ENiagara_AngleInput=::NewEnumerator1
    Cone Axis=<0 1 0>
    Module.UseConeFalloff=false
    Module.UseVelDistribution=false
    Override Seed=false
    Velocity Mode=::NewEnumerator2
    Velocity Speed=&RandomRangeFloat(Minimum=10, Maximum=50, Evaluation Type=::NewEnumerator1)
   )
  ]
  particleUpdateScript [
   &ParticleState()
   &ScaleSpriteSize(
    Scale Sprite Size Mode=::NewEnumerator0
    Uniform Scale Factor=&FloatFromCurve(FloatCurve=DataInterface() {propertyValues {Curve {keys [
         {arriveTangent=1.4511, interpMode=RCIM_Cubic, leaveTangent=1.4511, tangentMode=RCTM_User, time=-0.0013, value=0.1965}
         {arriveTangent=0.3947, interpMode=RCIM_Cubic, leaveTangent=0.3947, time=0.6257, value=0.6175}
         <at 1 0.5917>
        ]}}})
   )
   &DynamicMaterialParameters(
    Index 0 Param 1=&FloatFromCurve(FloatCurve=DataInterface() {propertyValues {Curve {keys [<at 0.5 1 RCIM_Cubic>,<at 1 -1>]}}})
    Index 0 Param 2=1
    Index 0 Param 3=1
    Index 0 Param 4=1
    Module.Param0WriteEnabled=true
    Module.Param1WriteEnabled=true
    Module.Param2WriteEnabled=true
    Module.Param3WriteEnabled=true
   )
   &Drag(Use Linear Drag=true, Use Rotational Drag=true)
   &SolveForcesAndVelocity(Clamp Velocity=false, Limit Acceleration=false)
   &ScaleColor(
    Scale Alpha=&FloatFromCurve(FloatCurve=DataInterface() {propertyValues {Curve {keys [
         <at 0.0015 0.6063 RCIM_Cubic>
         {arriveTangent=-0.2322, interpMode=RCIM_Cubic, leaveTangent=-0.2322, time=0.6018, value=0.5967}
         {arriveTangent=-3.0081, interpMode=RCIM_Cubic, leaveTangent=-3.0081, tangentMode=RCTM_User, time=0.8553, value=0.408}
         <at 1 0>
        ]}}})
    Scale RGB=<0 0 60>
    ScaleA=true
    ScaleRGB=true
   )
  ]
  rendererClasses [&NiagaraSpriteRendererProperties]
  renderers [{rendererClass=&NiagaraSpriteRendererProperties}]
 }
 background001 {
  emitterUpdateScript [
   &EmitterState(
    Inactive Response=::NewEnumerator1
    Life Cycle Mode=::NewEnumerator1
    Loop Behavior=::NewEnumerator1
    Loop Duration Mode=::NewEnumerator1
    UseLoopDelay=true
   )
   &SpawnBurst_Instantaneous(Spawn Count=7, Use Spawn Probability=false, UseLoopCountLimit=false)
  ]
  particleSpawnScript [
   &PartitionParticles(Number Of Partitions=6)
   &InitializeParticle(
    Color=&MakeLinearColorFromVectorAndFloat(
     "Vector (RGB)"=&MergeHelper_VectorSwitch(
      0=<0 0.1325 0.3594>
      1=<0 0.3608 1>
      2=<0 0.6629 1>
      3=<0.2901 0 1>
      4=<1 0.1067 0>
      5=<1 0.0197 0>
      Index=$Particles.PartitionParticles.Partition
      Num=::NewEnumerator4
     )
     "Float (Alpha)"=&MergeHelper_FloatSwitch(0=3, 1=1.5, 2=5, 3=5, 4=1, 5=1, Index=$Particles.PartitionParticles.Partition, Num=::NewEnumerator4)
    )
    Mesh Scale=&MergeHelper_VectorSwitch(
     0=<0.12 0.12 0.21>
     1=<0.05 0.05 0.18>
     2=<0.1 0.1 0.2>
     3=<0.1 0.1 0.15>
     4=<0.1 0.1 0.27>
     5=<0.1 0.1 0.3>
     Index=$Particles.PartitionParticles.Partition
     Num=::NewEnumerator4
    )
    Mesh Scale Mode=::NewEnumerator4
    UsePositionOffset=false
    Write Material Random=true
   )
   &SocketLocation(SamplePosOffset=false, Socket Reader=$System.Socket Reader, Socket Sampling Mode=::NewEnumerator3, VelocitySpeedLimit=false)
   /SetVariables(Particles.VisibilityTag=$Particles.PartitionParticles.Partition) {isSetParameters=true}
  ]
  particleUpdateScript [
   &ParticleState(Kill Particles When Lifetime Has Elapsed=false)
   &ScaleMeshSize(Scale Factor=<0.8 0.8 0.5>)
   &ScaleColor(
    Scale Alpha=&MergeHelper_FloatSwitch(0=0.5, 1=1, 2=0.5, 3=0.5, 4=1, 5=1, Index=$Particles.Initial.PartitionParticles.Partition, Num=::NewEnumerator4)
    Scale RGB=&VectorFromFloat(Value=&MergeHelper_FloatSwitch(0=1, 1=200, 2=1.5, 3=5, 4=70, 5=20, Index=$Particles.Initial.PartitionParticles.Partition, Num=::NewEnumerator4))
    ScaleA=true
    ScaleRGB=true
   )
   &DynamicMaterialParameters(
    Index 0 Param 2=1
    Index 0 Param 3=1
    Index 0 Param 4=1
    Module.Param0WriteEnabled=true
    Module.Param1WriteEnabled=true
    Module.Param2WriteEnabled=true
    Module.Param3WriteEnabled=true
   )
  ]
  rendererClasses [&NiagaraMeshRendererProperties]
  renderers [
   {rendererClass=&NiagaraMeshRendererProperties, rendererIndex=0}
   {rendererClass=&NiagaraMeshRendererProperties, rendererIndex=1}
   {rendererClass=&NiagaraMeshRendererProperties, rendererIndex=2}
   {rendererClass=&NiagaraMeshRendererProperties, rendererIndex=3}
   {rendererClass=&NiagaraMeshRendererProperties, rendererIndex=4}
   {rendererClass=&NiagaraMeshRendererProperties, rendererIndex=5}
  ]
 }
}
summary {
 systemName=NS_Pickaxe_Arctic_Sniper_Idle
 emitters [
  {emitterName=Fire, rendererClasses [&NiagaraSpriteRendererProperties]}
  {emitterName=background001, rendererClasses [&NiagaraMeshRendererProperties]}
 ]
 userVariables []
}
