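# Physics asset for the ArcticSniper two-handed pickaxe mesh, letting its hanging straps swing as the weapon moves.
# A capsule body on the handle bone anchors box bodies on two big-strap bones and one small-strap bone.
# The big straps form a two-link chain from the handle and the small strap attaches directly to it.
# All joints share moderate swing limits and a tight twist limit.
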PhysicsAsset /BRCosmetics/Weapons/FORT_Melee/Pickaxe_ArcticSniper/Meshes/Pickaxe_M_MED_ArcticSniper_Physics
bodies {
 Melee_TwoHanded_Handle [{shapeType=Capsule, length=73.6034, radius=1.8736, center=<0.4353 -1.2181 21.2414>, rotation=<rot 0 -90 0>}]
 dyn_bigStrap_01 [{shapeType=Box, extentX=8, extentY=2, extentZ=10, center=<4 0 0>}]
 dyn_bigStrap_02 [{shapeType=Box, extentX=8, extentY=2, extentZ=10, center=<4 0 0>}]
 dyn_smallStrap_01 [{shapeType=Box, extentX=8, extentY=10, extentZ=2, center=<4 0 0>}]
}
constraints [
 {bone1Name=dyn_bigStrap_01, bone2Name=Melee_TwoHanded_Handle, swing1LimitDegrees=40, swing2LimitDegrees=40, twistLimitDegrees=5}
 {bone1Name=dyn_bigStrap_02, bone2Name=dyn_bigStrap_01, swing1LimitDegrees=40, swing2LimitDegrees=40, twistLimitDegrees=5}
 {bone1Name=dyn_smallStrap_01, bone2Name=Melee_TwoHanded_Handle, swing1LimitDegrees=40, swing2LimitDegrees=40, twistLimitDegrees=5}
]
