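# Animation Blueprint for the ArcticSniper pickaxe skeleton that adds secondary physics motion to the melee weapon mesh.
# The input pose passes through a single RigidBody node using the pickaxe's physics asset, with overridden world gravity and clamped component acceleration, then goes to the output.
# Anim Dynamics nodes for the big and small strap bones remain in the graph but do not feed the output.
# The event graph holds no logic.
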AnimBlueprint /BRCosmetics/Weapons/FORT_Melee/Pickaxe_ArcticSniper/Meshes/Pickaxe_M_MED_ArcticSniper_Skeleton_AnimBlueprint
target_skeleton=&/Pickaxe_M_MED_ArcticSniper_Skeleton
graphs [
 AnimGraph {nodes [
   Anim Dynamics - Bone: dyn_bigStrap_01(Alpha=1, ExternalForce="0,0,0") {
    bGravityOverrideInSimSpace=true
    alphaBoolBlend {blendOption=Linear}
    boundBone {boneName=dyn_bigStrap_01}
    physicsBodyDefinitions [{
      sphereCollisionRadius=0
      boundBone {boneName=dyn_bigStrap_01}
      boxExtents=<1 1 1>
      constraintSetup {angularLimitsMax=<0 0 0>, angularLimitsMin=<0 0 0>, angularTarget=<0 0 0>}
      localJointOffset=<0 0 0>
     }]
   }
   n1 = Anim Dynamics - Bone: dyn_bigStrap_01(Alpha=0.75, ExternalForce="0,0,0") {
    angularSpringConstant=30000
    bGravityOverrideInSimSpace=true
    lODThreshold=1
    alphaBoolBlend {blendOption=Linear}
    boundBone {boneName=dyn_bigStrap_01}
    componentAppliedLinearAccClamp=<10000 10000 10000>
    componentLinearAccScale=<1.2 1.2 1.2>
    componentLinearVelScale=<1.1 1.1 1.1>
    physicsBodyDefinitions [{
      sphereCollisionRadius=0
      boundBone {boneName=dyn_bigStrap_01}
      boxExtents=<20 75 80>
      constraintSetup {angularLimitsMax=<90 90 90>, angularLimitsMin=<-90 -90 -90>, angularTarget=<0 0 0>}
      localJointOffset=<15 0 0>
     }]
   }
   n2 = Anim Dynamics - Bone: dyn_bigStrap_02(ComponentPose=@n1, Alpha=0.75, ExternalForce="0,0,0") {
    angularSpringConstant=30000
    bGravityOverrideInSimSpace=true
    lODThreshold=1
    alphaBoolBlend {blendOption=Linear}
    boundBone {boneName=dyn_bigStrap_02}
    componentAppliedLinearAccClamp=<10000 10000 10000>
    componentLinearAccScale=<1.2 1.2 1.2>
    componentLinearVelScale=<1.1 1.1 1.1>
    physicsBodyDefinitions [{
      sphereCollisionRadius=0
      boundBone {boneName=dyn_bigStrap_02}
      boxExtents=<20 75 80>
      constraintSetup {angularLimitsMax=<90 90 90>, angularLimitsMin=<-90 -90 -90>, angularTarget=<0 0 0>}
      localJointOffset=<15 0 0>
     }]
   }
   Anim Dynamics - Bone: dyn_smallStrap_01(ComponentPose=@n2, Alpha=0.75, ExternalForce="0,0,0") {
    angularSpringConstant=30000
    bGravityOverrideInSimSpace=true
    lODThreshold=1
    alphaBoolBlend {blendOption=Linear}
    boundBone {boneName=dyn_smallStrap_01}
    componentAppliedLinearAccClamp=<10000 10000 10000>
    componentLinearAccScale=<1.2 1.2 1.2>
    componentLinearVelScale=<1.1 1.1 1.1>
    physicsBodyDefinitions [{
      sphereCollisionRadius=0
      boundBone {boneName=dyn_smallStrap_01}
      boxExtents=<10 30 30>
      constraintSetup {angularLimitsMax=<90 90 90>, angularLimitsMin=<-90 -90 -90>, angularTarget=<0 0 0>}
      localJointOffset=<7 0 0>
     }]
   }
   AnimGraphNode_SubInput_0 = Input Pose()
   n3 = RigidBody(ComponentPose=Local To Component(LocalPose=@AnimGraphNode_SubInput_0)) {
    bOverrideWorldGravity=true
    overridePhysicsAsset=&/Pickaxe_M_MED_ArcticSniper_Physics
    alphaBoolBlend {blendOption=Linear}
    componentAppliedLinearAccClamp=<4000 4000 4000>
    componentLinearAccScale=<1 1 1>
    componentLinearVelScale=<1 1 1>
    overrideWorldGravity {z=-980}
   }
   Output Pose(Result=Component To Local(ComponentPose=@n3)) {layerGroup=None}
  ]}
 EventGraph {nodes [
   TryGetPawnOwner()
   Event BlueprintUpdateAnimation()
  ]}
]
variables []
